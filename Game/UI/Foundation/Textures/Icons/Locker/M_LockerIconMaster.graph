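# UI master material for locker icons, translucent, that tints a parameterized icon texture, defaulting to the outfit icon.
# The texture's red channel blends between two color parameters, from RGBA_GreenChannel to RGBA_RedChannel; the green channel itself is not read.
# The same mask blends the two colors' alphas, and the texture's alpha scales that result.
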Material /Game/UI/Foundation/Textures/Icons/Locker/M_LockerIconMaster
graph {
 properties {
  bUsedWithStaticMesh=false
  blendMode=BLEND_Translucent
  materialDomain=MD_UI
  pixelDepthOffsetMode=PDOM_Legacy
  usedShadingModels=Default Lit
 }
 nodes [
  n1 = VectorParameter() {parameterName=RGBA_GreenChannel, defaultValue=<rgba 0.5661 0.8854 0.4242 1>}
  n2 = VectorParameter() {parameterName=RGBA_RedChannel, defaultValue=<rgba 0.8177 0.2362 0.2851 1>}
  n3 = TextureSampleParameter2D() {parameterName=IconTexture, texture=&/T_Icon_Outfit_128}
  LinearInterpolate(A=@n1.RGB, B=@n2.RGB, Alpha=@n3.R)
  LinearInterpolate(B=LinearInterpolate(A=@n1.A, B=@n2.A, Alpha=@n3.R), Alpha=@n3.A)
 ]
}
